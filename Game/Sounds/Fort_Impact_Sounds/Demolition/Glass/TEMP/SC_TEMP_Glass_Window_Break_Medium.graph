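# Temporary sound cue for a medium glass window breaking, in the demolition impact sound set.
# A Random node picks one of eight equally weighted break waves, then a Modulator adds slight pitch and volume variation.
# It routes to the physics impact sound class with its own attenuation, global and local concurrency limits, and virtualization disabled.
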SoundCue /Game/Sounds/Fort_Impact_Sounds/Demolition/Glass/TEMP/SC_TEMP_Glass_Window_Break_Medium
root=Modulator_0
nodes [
 {id=Modulator_0, children [Random_1], properties {pitchMax=1.1, pitchMin=0.9, volumeMax=1, volumeMin=0.9}}
 {
  id=Random_1
  children [WavePlayer_8, WavePlayer_9, WavePlayer_10, WavePlayer_11, WavePlayer_12, WavePlayer_13, WavePlayer_14, WavePlayer_15]
  properties {bShouldExcludeFromBranchCulling=true, weights [1,1,1,1,1,1,1,1]}
 }
 {id=WavePlayer_10, properties {soundWaveAssetPtr=&/TEMP_Glass_Window_Break_Medium_03}}
 {id=WavePlayer_11, properties {soundWaveAssetPtr=&/TEMP_Glass_Window_Break_Medium_04}}
 {id=WavePlayer_12, properties {soundWaveAssetPtr=&/TEMP_Glass_Window_Break_Medium_05}}
 {id=WavePlayer_13, properties {soundWaveAssetPtr=&/TEMP_Glass_Window_Break_Medium_06}}
 {id=WavePlayer_14, properties {soundWaveAssetPtr=&/TEMP_Glass_Window_Break_Medium_07}}
 {id=WavePlayer_15, properties {soundWaveAssetPtr=&/TEMP_Glass_Window_Break_Medium_08}}
 {id=WavePlayer_8, properties {soundWaveAssetPtr=&/TEMP_Glass_Window_Break_Medium_01}}
 {id=WavePlayer_9, properties {soundWaveAssetPtr=&/TEMP_Glass_Window_Break_Medium_02}}
]
properties {
 attenuationSettings=&/ATT_TEMP_Glass_Window_Break_Medium
 maxDistance=12400
 soundClassObject=/Game/Sounds/FortSoundClassesAndModes/Fort_Physics_Impact
 virtualizationMode=Disabled
 concurrencySet [
  &/CONC_TEMP_Glass_Window_Break_Global
  &/CONC_TEMP_Glass_Window_Break_Local
 ]
}
